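# Additive unlit particle sprite material for a flare with a hot center and no time-driven flicker.
# A UV-centered soft sphere mask is raised to powers for a tight, brightened, partly whitened core and a faint wider glow.
# The glow joins the FlareMask texture, is tinted by particle color, then adds to the core.
# Particle alpha passes through a parameterized depth fade, which a shading path switch skips on mobile.
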Material /Figment_S06/Effects/Fort_Effects/Materials/Flares/M_Particle_Flare_HotCenter_NoFlicker
graph {
 properties {bUsedWithParticleSprites=true, blendMode=BLEND_Additive, pixelDepthOffsetMode=PDOM_Legacy, shadingModel=MSM_Unlit, usedShadingModels=Unlit}
 nodes [
  n1 = Constant() {r=0.5}
  n2 = SphereMask(A=TextureCoordinate(), B=@n1, Radius=@n1) {hardnessPercent=0}
  n3 = Power(Base=@n2) {constExponent=25}
  n4 = ParticleColor()
  n5 = Multiply(A=Power(Base=@n2) {constExponent=4}) {constB=0.15}
  n6 = TextureSampleParameter2D() {parameterName=FlareMask, texture=/Game/Effects/Fort_Effects/Textures/Flares/T_Flare_Spikeball}
  Add(
   A=Multiply(A=Multiply(A=@n3) {constB=2}, B=LinearInterpolate(B=@n4.RGB) {constA=1, constAlpha=0.7})
   B=Multiply(A=Add(A=@n5, B=@n6.RGB), B=@n4.RGB)
  )
  n7 = DepthFade(Opacity=@n4.A, FadeDistance=ScalarParameter() {defaultValue=30, parameterName=DepthFade})
  Multiply(A=Constant() {r=1}, B=ShadingPathSwitch(Default=@n7, Deferred=@n7, Mobile=@n4.A) {desc=Migrated from FeatureLevelSwitch})
 ]
}
